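# Master deferred-decal material for Kurai floor decals, with two tintable line colors and a translucent blend.
# World-aligned noise blends color and roughness, while a "Use Texture?" switch picks texture-driven opacity and flattened normal, or contrasted dual-noise opacity and flat normal.
# Vertex alpha scales opacity, but Low and Medium quality use only the constant max opacity.
# The attributes then pass through NeonCity's MF_WaterPuddlesBuilding function.
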Material /TimberStakeMapContent/Environments/Props/FloorDecals/Materials/MasterMaterials/M_Kurai_FloorDecals_Master_A_A
graph {
 properties {
  bAutomaticallySetUsageInEditor=false
  bCastRayTracedShadows=false
  bUseMaterialAttributes=true
  bUsedWithNanite=true
  blendMode=BLEND_Translucent
  materialDomain=MD_DeferredDecal
  usedShadingModels=Default Lit
 }
 nodes [
  Constant()
  n1 = VectorParameter() {parameterName=Line Color, defaultValue=<rgba 1 0.5424 0 1>}
  n2 = VectorParameter() {parameterName=Line Color_B, defaultValue=<rgba 0.8906 0.4831 0 1>}
  n3 = ScalarParameter() {defaultValue=512, parameterName=Noise tiling}
  n4 = MaterialFunctionCall(
   TextureObject=TextureObject() {texture=/TimberStakeMapContent/Environments/Props/FloorDecals/Textures/T_Kurai_FloorDecals_Noise_A_A}
   TextureSize=@n3
  ) {materialFunction=&WorldAlignedTexture}
  n5 = ScalarParameter() {defaultValue=0.4, parameterName=Roughness Min}
  n6 = ScalarParameter() {defaultValue=0.6, parameterName=Roughness Max}
  n7 = ScalarParameter() {defaultValue=1, parameterName=Opacity_Max, sliderMax=1}
  n8 = TextureSampleParameter2D() {parameterName=Opacity Mask, texture=/TimberStakeMapContent/Environments/Textures/Decals/T_Kurai_CautionDecal_A_A}
  n9 = TextureObject() {samplerType=SAMPLERTYPE_Masks, texture=/Game/Characters/Player/Male/Large/Bodies/M_LRG_Clash_V/Textures/Simple_Perlin_M}
  n10 = MaterialFunctionCall(
   In=Add(
    A=MaterialFunctionCall(TextureObject=@n9, TextureSize=@n3) {materialFunction=&WorldAlignedTexture}.XYZ Texture
    B=@n4.XYZ Texture
   )
   Contrast=ScalarParameter() {defaultValue=-0.75, parameterName=Noise Contrast}
  ) {materialFunction=&CheapContrast}
  n11 = Multiply(
   A=VertexColor().A
   B=LinearInterpolate(
    A=ScalarParameter() {parameterName=Opacity_Min}
    B=@n7
    Alpha=StaticSwitchParameter(True=@n8.RGB, False=@n10.Result) {parameterName=Use Texture?}
   )
  )
  n12 = TextureSampleParameter2D() {
   parameterName=Normal Texture
   samplerType=SAMPLERTYPE_Normal
   texture=/TimberStakeMapContent/Environments/Textures/Decals/T_Kurai_CautionDecal_A_A_N
  }
  n13 = MaterialFunctionCall(Normal=@n12.RGB, Flatness=ScalarParameter() {defaultValue=0.75, parameterName=Normal Flatness}) {materialFunction=&FlattenNormal}
  n14 = TextureSample() {samplerType=SAMPLERTYPE_Normal, texture=&FlatNormalBlue}
  MaterialFunctionCall!(In=MakeMaterialAttributes!(
    BaseColor=LinearInterpolate(A=@n1.RGB, B=@n2.RGB, Alpha=@n4.XYZ Texture)
    Metallic
    Specular
    Roughness=LinearInterpolate(A=@n5, B=@n6, Alpha=@n4.XYZ Texture)
    Anisotropy
    EmissiveColor
    Opacity=QualitySwitch(Default=@n11, Low=@n7, High=@n11, Medium=@n7, Epic=@n11)
    OpacityMask
    Normal=StaticSwitchParameter(True=@n13.Result, False=@n14.RGB) {parameterName=Use Texture?}
    Tangent
    WorldPositionOffset
    SubsurfaceColor
    ClearCoat
    ClearCoatRoughness
    AmbientOcclusion
    Refraction
    CustomizedUVs_0
    CustomizedUVs_1
    CustomizedUVs_2
    CustomizedUVs_3
    CustomizedUVs_4
    CustomizedUVs_5
    CustomizedUVs_6
    CustomizedUVs_7
    PixelDepthOffset
    ShadingModel
    Displacement
   )) {materialFunction=/Game/Environments/Asteria/Sets/NeonCity/MaterialFunctions/MF_WaterPuddlesBuilding}
 ]
}
